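; PCG graph that measures the area of the shapes passed to its Areas input, outputting the values as attributes.
; Polyline data becomes 2D polygons through spline control points and Create Polygon 2D, while existing Polygon2D data is used directly.
; Each polygon's built-in area data property is extracted as an Area attribute, then merged and sorted.
; A final Multiply op, with only its first operand wired, produces the output.
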(graph MeasureAreas
  (inputs (Areas :type Any))
  (bind FilterDataByType_0 (Filter%20Data%20By%20Type :TargetType (struct :Ids [(struct :Struct /Script/PCG.PCGDataTypeInfoPolyline)])))
  (bind GetSplineControlPoints_0 (Get%20Spline%20Control%20Points))
  (bind CreatePolygon2D_0 (Create%20Polygon%202D))
  (bind FilterDataByType_1 (Filter%20Data%20By%20Type :TargetType (struct :Ids [(struct :Struct /Script/PCG.PCGDataTypeInfoPolygon2D)])))
  (bind ExtractAttribute_0 (Extract%20Attribute%20at%20Index :bForceOutputAttributeToBeInElementsDomain true :InputSource @data.$area :OutputAttributeName Area))
  (bind MergeAttributes_0 (Merge%20Attributes))
  (bind SortAttributes_0 (Sort%20Attributes :bUseStableSort false))
  (bind AttributeMathsOp_0 (Attribute%20Maths%20Op :Operation Multiply))
  (connect SortAttributes_0 Out AttributeMathsOp_0 InA)
  (connect GetSplineControlPoints_0 Out CreatePolygon2D_0 In)
  (connect CreatePolygon2D_0 Out ExtractAttribute_0 In)
  (connect FilterDataByType_1 InsideFilter ExtractAttribute_0 In)
  (connect DefaultInputNode Areas FilterDataByType_0 In)
  (connect DefaultInputNode Areas FilterDataByType_1 In)
  (connect FilterDataByType_0 InsideFilter GetSplineControlPoints_0 In)
  (connect ExtractAttribute_0 Out MergeAttributes_0 In)
  (connect MergeAttributes_0 Out SortAttributes_0 In)
  (output Out AttributeMathsOp_0))
