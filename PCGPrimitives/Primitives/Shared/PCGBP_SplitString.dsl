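; PCG Blueprint node that splits a string attribute into segments, though its title override reads "Join Strings".
; For each input attribute set, it reads the target attribute row by row and parses it on the configured separator.
; Each segment becomes an entry in a new attribute set, storing the text, source row index and a "Type" code.
; Type comes from substring matches on Theme, Size, Layout or Variation, otherwise -1.
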(fn ExecuteWithContext
  (params (InContext :type PCGContext%20Structure%20%28by%20ref%29) (Input :type PCGData%20Collection%20Structure%20%28by%20ref%29))
  (bind _inmetadata (LocalVariables|GetInMetadata))
  (bind _returnvalue (PCG|Metadata|GetNumberofEntries _inmetadata))
  (bind _returnvalue_2 (- _returnvalue 1))
  (bind _returnvalue_3 (Data|GetParams Input))
  (for _array_element _returnvalue_3
    (LocalVariables|SetPCGData _array_element)
    (Utilities|FlowControl|Sequence
      (:then_0
        (bind _returnvalue_4 (Game|ConstructObjectfromClass "/Script/PCG.PCGParamData"))
        (bind _output_get (LocalVariables|SetPCGAttribSet _returnvalue_4))
        (bind _returnvalue_5 (Metadata|MutableMetadata _output_get))
        (LocalVariables|SetOutMetadata _returnvalue_5)
        (PCG|Metadata|CreateStringAttribute (LocalVariables|GetOutMetadata) (Variables|Default|GetOutAttributeName) "" false true)
        (PCG|Metadata|CreateInteger32Attribute (LocalVariables|GetOutMetadata) (Variables|Default|GetRowAttributeName) 0 false true)
        (PCG|Metadata|CreateInteger32Attribute (LocalVariables|GetOutMetadata) "Type" 0 false true))
      (:then_1
        (bind _aspcgattribute_set (Utilities|Casting|CastToPCGParamData (PCG|TaggedData|BreakTaggedData (LocalVariables|GetPCGData))))
        (bind _output_get_2 (LocalVariables|SetInMetadata (Metadata|ConstMetadata _aspcgattribute_set)))
        (if (and (PCG|Metadata|HasAttribute _output_get_2 (Variables|Default|GetTargetAttribute)) (!= _returnvalue_2 -1))
          (for _index (range (+ (Math|Conversions|ToInteger(Integer64) _returnvalue_2) 1))
            (LocalVariables|SetCurrentRow _index)
            (bind _returnvalue_6 (PCG|Metadata|GetStringAttributebyMetadataKey (Math|Conversions|ToInteger64(Integer) _index) (LocalVariables|GetInMetadata) (Variables|Default|GetTargetAttribute)))
            (for _array_element_2 (Utilities|String|ParseIntoArray _returnvalue_6 (Variables|Default|GetSeparator))
              (bind _output_get_3 (LocalVariables|SetCurrentStringSegment _array_element_2))
              (if (Utilities|String|Contains _output_get_3 "Theme")
                (LocalVariables|SetType 0)
                (elif (Utilities|String|Contains _output_get_3 "Size")
                  (LocalVariables|SetType 1)
                  (elif (Utilities|String|Contains _output_get_3 "Layout")
                    (LocalVariables|SetType 2)
                    (elif (Utilities|String|Contains _output_get_3 "Variation")
                      (LocalVariables|SetType 3)
                      (else
                        (LocalVariables|SetType -1))))))
              (bind _returnvalue_7 (PCG|Metadata|AddEntry (LocalVariables|GetOutMetadata) -1))
              (LocalVariables|SetOutMetaDataEntry _returnvalue_7)
              (PCG|Metadata|SetStringAttributebyMetadataKey (LocalVariables|GetOutMetaDataEntry) (LocalVariables|GetOutMetadata) (Variables|Default|GetOutAttributeName) (LocalVariables|GetCurrentStringSegment))
              (PCG|Metadata|SetInteger32AttributebyMetadataKey (LocalVariables|GetOutMetaDataEntry) (LocalVariables|GetOutMetadata) (Variables|Default|GetRowAttributeName) (LocalVariables|GetCurrentRow))
              (PCG|Metadata|SetInteger32AttributebyMetadataKey (LocalVariables|GetOutMetaDataEntry) (LocalVariables|GetOutMetadata) "Type" (LocalVariables|GetType))))
          (else
            (LocalVariables|SetOutString "INVALID"))))
      (:then_2
        (Utilities|Array|Add (LocalVariables|GetOutData) (PCG|TaggedData|MakeTaggedData (LocalVariables|GetPCGAttribSet))))))
  (return (Utilities|Struct|MakePCGDataCollection (LocalVariables|GetOutData))))

(fn NodeColorOverride
  (return "(R=1.000000,G=0.600000,B=0.000000,A=1.000000)"))

(fn NodeTitleOverride
  (return "Join Strings"))
