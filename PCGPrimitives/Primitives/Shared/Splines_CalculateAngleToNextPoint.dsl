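; PCG primitive subgraph that writes an AngleToNextPoint attribute onto input spline data from the yaw difference between neighbouring control points.
; It assigns each element a wrapped next-point index, copies local yaw to Ryaw, then sorts by that index to produce a one-step-shifted copy for subtraction.
; A divide, truncate and multiply-add chain normalizes the clockwise/counter-clockwise difference before copying it back to the indexed input.
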(graph Splines_CalculateAngleToNextPoint
  (bind AttributeMathsOp_8 (Attribute%20Maths%20Op :InputSource1 $Index :InputSource3 @Data.$NumElements :Operation AddModulo :OutputTarget nextPointindex))
  (bind GetSplineControlPoints_0 (Get%20Spline%20Control%20Points))
  (bind CopyAttributes_2 (Copy%20Attributes :InputSource $LocalRotation.Yaw :OutputTarget Ryaw))
  (bind SortAttributes_0 (Sort%20Attributes :bUseStableSort false :SortAttributes [(struct :InputSource nextPointIndex)]))
  (bind AttributeMathsOp_2 (Attribute%20Maths%20Op :InputSource1 Ryaw :InputSource2 Ryaw :Operation Subtract))
  (bind CopyAttributes_1 (Copy%20Attributes :InputSource ryaw :OutputTarget AngleToNextPoint))
  ;[0 0 0 0.5] Normalize CW/CCW
  (bind AttributeMathsOp_0 (Attribute%20Maths%20Op :InputSource1 Ryaw :Operation Divide :OutputTarget NormalizingFactor))
  (bind AttributeMathsOp_1 (Attribute%20Maths%20Op :bForceRoundingOpToInt true :InputSource1 NormalizingFactor :Operation Truncate))
  (bind AttributeMathsOp_7 (Attribute%20Maths%20Op :InputSource1 Ryaw :InputSource2 NormalizingFactor :Operation MulAdd))
  (connect AttributeMathsOp_2 Out AttributeMathsOp_0 InA)
  (connect AttributeMathsOp_0 Out AttributeMathsOp_1 In)
  (connect CopyAttributes_2 Out AttributeMathsOp_2 InA)
  (connect SortAttributes_0 Out AttributeMathsOp_2 InB)
  (connect AttributeMathsOp_1 Out AttributeMathsOp_7 InA)
  (connect AttributeMathsOp_1 Out AttributeMathsOp_7 InB)
  (connect DefaultInputNode In AttributeMathsOp_8 InA)
  (connect DefaultInputNode In AttributeMathsOp_8 InC)
  (connect AttributeMathsOp_7 Out CopyAttributes_1 Source)
  (connect AttributeMathsOp_8 Out CopyAttributes_1 Target)
  (connect AttributeMathsOp_8 Out CopyAttributes_2 Source)
  (connect GetSplineControlPoints_0 Out CopyAttributes_2 Target)
  (connect AttributeMathsOp_8 Out GetSplineControlPoints_0 In)
  (connect CopyAttributes_2 Out SortAttributes_0 In)
  (output Out CopyAttributes_1))
