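; PCG graph that trims a `distance` parameter's length off both ends of open splines.
; It splits by distance, keeps the tail piece, reverses it and repeats, leaving output splines reversed.
; Tail selection sorts piece indices descending, deduplicates on OriginatingSplineIndex, then joins the survivors for Filter Data By Index.
; Data lacking that attribute after either split bypasses the remaining steps straight to the output.
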(graph Trim_open_Spline_each_side
  (params (distance :type Double :default 5e+01))
  (bind DeleteAttributes_2 (Delete%20Attributes :MetadataDomain Data :SelectedAttributes "OriginatingSplineIndex"))
  (bind SplitSpline_0 (Split%20Spline :Mode ByDistance :Constant-param distance)) ; Trim `distance` from one end of each open spline
  (bind FilterDataByAttribute_0 (Filter%20Data%20By%20Attribute :Attribute @Data.OriginatingSplineIndex))
  (bind FilterDataByIndex_2 (Filter%20Data%20By%20Index)) ; Per-spline: keep the segment AFTER the cut\n(drop the head piece)
  (bind SplineDirection_2 (Spline%20Direction)) ; Reverse so we can trim the other end\nwith the same pattern
  (bind SplitSpline_2 (Split%20Spline :Mode ByDistance :Constant-param distance))
  (bind FilterDataByAttribute_2 (Filter%20Data%20By%20Attribute :Attribute @Data.OriginatingSplineIndex))
  (bind FilterDataByIndex_3 (Filter%20Data%20By%20Index)) ; Keep the remaining middle (both ends now trimmed)
  (bind DeleteAttributes_3 (Delete%20Attributes :MetadataDomain Data :SelectedAttributes "OriginatingSplineIndex"))
  ;[0 0 0 0.5] Pick Remainder
  (bind ExtractAttribute_0 (Extract%20Attribute%20at%20Index :bForceOutputAttributeToBeInElementsDomain true :InputSource @Data.OriginatingSplineIndex))
  (bind MergeAttributes_0 (Merge%20Attributes))
  (bind CopyAttributes_0 (Copy%20Attributes :InputSource $Index :OutputTarget IndexBck))
  (bind SortAttributes_0 (Sort%20Attributes :bUseStableSort false :SortAttributes [(struct :InputSource IndexBck :SortMethod Descending)]))
  (bind AttributeRemoveDuplicates_1 (Attribute%20Remove%20Duplicates :AttributeSelectors [OriginatingSplineIndex]))
  (bind SortAttributes_1 (Sort%20Attributes :bUseStableSort false :SortAttributes [(struct :InputSource IndexBck)]))
  (bind AttributeReduce_0 (Attribute%20Reduce :InputSource IndexBck :Operation Join))
  (bind ExtractAttribute_1 (Extract%20Attribute%20at%20Index :bForceOutputAttributeToBeInElementsDomain true :InputSource @Data.OriginatingSplineIndex))
  (bind MergeAttributes_1 (Merge%20Attributes))
  (bind CopyAttributes_1 (Copy%20Attributes :InputSource $Index :OutputTarget IndexBck))
  (bind SortAttributes_2 (Sort%20Attributes :bUseStableSort false :SortAttributes [(struct :InputSource IndexBck :SortMethod Descending)]))
  (bind AttributeRemoveDuplicates_2 (Attribute%20Remove%20Duplicates :AttributeSelectors [OriginatingSplineIndex]))
  (bind SortAttributes_3 (Sort%20Attributes :bUseStableSort false :SortAttributes [(struct :InputSource IndexBck)]))
  (bind AttributeReduce_1 (Attribute%20Reduce :InputSource IndexBck :Operation Join))
  (connect SortAttributes_1 Out AttributeReduce_0 In)
  (connect SortAttributes_3 Out AttributeReduce_1 In)
  (connect SortAttributes_0 Out AttributeRemoveDuplicates_1 In)
  (connect SortAttributes_2 Out AttributeRemoveDuplicates_2 In)
  (connect MergeAttributes_0 Out CopyAttributes_0 Source)
  (connect MergeAttributes_0 Out CopyAttributes_0 Target)
  (connect MergeAttributes_1 Out CopyAttributes_1 Source)
  (connect MergeAttributes_1 Out CopyAttributes_1 Target)
  (connect DefaultInputNode In DeleteAttributes_2 In)
  (connect FilterDataByIndex_3 InsideFilter DeleteAttributes_3 In)
  (connect FilterDataByAttribute_0 InsideFilter ExtractAttribute_0 In)
  (connect FilterDataByAttribute_2 InsideFilter ExtractAttribute_1 In)
  (connect SplitSpline_0 Out FilterDataByAttribute_0 In)
  (connect SplitSpline_2 Out FilterDataByAttribute_2 In)
  (connect FilterDataByAttribute_0 InsideFilter FilterDataByIndex_2 In)
  (connect AttributeReduce_0 Out FilterDataByIndex_2 SelectedIndices)
  (connect FilterDataByAttribute_2 InsideFilter FilterDataByIndex_3 In)
  (connect AttributeReduce_1 Out FilterDataByIndex_3 SelectedIndices)
  (connect ExtractAttribute_0 Out MergeAttributes_0 In)
  (connect ExtractAttribute_1 Out MergeAttributes_1 In)
  (connect CopyAttributes_0 Out SortAttributes_0 In)
  (connect AttributeRemoveDuplicates_1 Out SortAttributes_1 In)
  (connect CopyAttributes_1 Out SortAttributes_2 In)
  (connect AttributeRemoveDuplicates_2 Out SortAttributes_3 In)
  (connect FilterDataByIndex_2 InsideFilter SplineDirection_2 In)
  (connect DeleteAttributes_2 Out SplitSpline_0 In)
  (connect SplineDirection_2 Out SplitSpline_2 In)
  (output Out DeleteAttributes_3)
  (output Out FilterDataByAttribute_0.OutsideFilter)
  (output Out FilterDataByAttribute_2.OutsideFilter))
