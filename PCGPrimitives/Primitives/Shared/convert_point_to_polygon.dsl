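; PCG graph that converts point data into a 2D polygon built from the corners of the point bounds.
; It filters the input to point data, then four Reset Point Center nodes move copies to the XY bounds corners at mid-height.
; Merge Points combines the corner points and feeds a closed linear spline, which Create Polygon 2D turns into the output.
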(graph convert_point_to_polygon
  (bind FilterDataByType_0 (Filter%20Data%20By%20Type :TargetType (struct :Ids [(struct :Struct /Script/PCG.PCGDataTypeInfoPoint)])))
  (bind ResetPointCenter_0 (Reset%20Point%20Center :PointCenterLocation (0.0 0.0 0.5)))
  (bind ResetPointCenter_1 (Reset%20Point%20Center :PointCenterLocation (0.0 1.0 0.5)))
  (bind ResetPointCenter_2 (Reset%20Point%20Center :PointCenterLocation (1.0 0.0 0.5)))
  (bind ResetPointCenter_3 (Reset%20Point%20Center :PointCenterLocation (1.0 1.0 0.5)))
  (bind MergePoints_0 (Merge%20Points))
  (bind CreateSpline_0 (Create%20Spline :bClosedLoop true :bLinear true :InterpTypeAttribute None))
  (bind CreatePolygon2D_0 (Create%20Polygon%202D))
  (connect CreateSpline_0 Out CreatePolygon2D_0 In)
  (connect MergePoints_0 Out CreateSpline_0 In)
  (connect DefaultInputNode In FilterDataByType_0 In)
  (connect ResetPointCenter_0 Out MergePoints_0 In)
  (connect ResetPointCenter_1 Out MergePoints_0 In)
  (connect ResetPointCenter_2 Out MergePoints_0 In)
  (connect ResetPointCenter_3 Out MergePoints_0 In)
  (connect FilterDataByType_0 InsideFilter ResetPointCenter_0 In)
  (connect FilterDataByType_0 InsideFilter ResetPointCenter_1 In)
  (connect FilterDataByType_0 InsideFilter ResetPointCenter_2 In)
  (connect FilterDataByType_0 InsideFilter ResetPointCenter_3 In)
  (output Out CreatePolygon2D_0))
